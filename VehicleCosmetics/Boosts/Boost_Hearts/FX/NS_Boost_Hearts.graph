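# Niagara vehicle boost cosmetic that trails hearts and swirling ribbons from the car's boost sockets.
# Hearts emitter spawns pink-to-purple SubUV sprites at filtered skeletal sockets, throwing them backward as they shrink.
# Ribbon emitter writes three sine/cosine-offset positions and extra colors for three ribbon renderers, creating tapering spirals.
# Both apply fading inherited source movement and GameModeID switches, mostly with identical values; painted-variant overrides are disabled.
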NiagaraSystem /VehicleCosmetics/Boosts/Boost_Hearts/FX/NS_Boost_Hearts
dependencies {
 usedDataInterfaces [
  &NiagaraDataInterfaceCurve
  &NiagaraDataInterfaceSpriteRendererInfo
  &NiagaraDataInterfaceSkeletalMesh
  &NiagaraDataInterfaceArrayFloat3
 ]
 usedDynamicInputs [&SelectVectorFromArray]
 usedModules [
  &self:SystemSpawnScript.SystemScriptSource.SystemScriptGraph.NiagaraNodeAssignment_0.SetVariables_#1
  &self:SystemSpawnScript.SystemScriptSource.SystemScriptGraph.NiagaraNodeAssignment_1.SetVariables_#2
  &self:SystemSpawnScript.SystemScriptSource.SystemScriptGraph.NiagaraNodeAssignment_2.SetVariables_#3
  &SystemState
 ]
}
emitters {
 Hearts {
  emitterUpdateScript [
   &EmitterState()
   &VC_FilteredSocketCount()
   &SpawnRate(SpawnRate=&RandomRangeFloat(Minimum=100, Maximum=150), Use Spawn Probability=false)
  ]
  particleSpawnScript [
   &VC_PaintedVariantsOverride() {enabled=false}
   &InitializeParticle(
    Color=&MergeHelper_ColorSwitch(
     0=&RandomRangeLinearColor(Minimum=<rgba 0.1624 0.0596 0.7917 1>, Maximum=<rgba 0.9375 0.0831 0.1279 1>, Color Channel Mode=::NewEnumerator0)
     1=&RandomRangeLinearColor(Minimum=<rgba 0.1624 0.0596 0.7917 1>, Maximum=<rgba 0.9375 0.0831 0.1279 1>, Color Channel Mode=::NewEnumerator0)
     2=&RandomRangeLinearColor(Minimum=<rgba 0.1624 0.0596 0.7917 1>, Maximum=<rgba 0.9375 0.0831 0.1279 1>, Color Channel Mode=::NewEnumerator0)
     3=&RandomRangeLinearColor(Minimum=<rgba 0.1624 0.0596 0.7917 1>, Maximum=<rgba 0.9375 0.0831 0.1279 1>, Color Channel Mode=::NewEnumerator0)
     Index=$System.GameModeID
     Num=::NewEnumerator2
    )
    Lifetime Max=1
    Lifetime Min=0.5
    Lifetime Mode=::NewEnumerator1
    Sprite Rotation Angle=&RandomRangeFloat(Minimum=-0.1, Maximum=0.1)
    Sprite Rotation Mode=::NewEnumerator1
    Sprite Size Mode=::NewEnumerator1
    Uniform Sprite Size Max=&Multiply_Float(A=13, B=3)
    Uniform Sprite Size Min=&Multiply_Float(A=22, B=3)
    UsePositionOffset=false
    Write Material Random=true
   )
   &PartitionParticles(Number Of Partitions=$Emitter.PortCount)
   &SkeletalMeshLocation(
    Bone / Socket Index=$Particles.PartitionParticles.Partition
    Bone / Socket Sampling Mode=::NewEnumerator5
    MaskSpawnGroup=false
    Mesh Sampling Type=::NewEnumerator6
    Sampled Position Offset=&AddVectorToPosition(Position=<-10 0 0>, Vector=$User.DTCarOffset)
    Skeletal Mesh=$System.SkeletalMesh
   )
   &AddVelocity(Velocity=&MergeHelper_VectorSwitch(
     0=&RandomRangeVector(Minimum=<-800 -200 -50>, Maximum=<-1000 200 300>)
     1=&RandomRangeVector(Minimum=<-800 -200 -50>, Maximum=<-1000 200 300>)
     2=&RandomRangeVector(Minimum=<-800 -200 -50>, Maximum=<-1000 200 300>)
     3=&RandomRangeVector(Minimum=<-800 -200 -50>, Maximum=<-1000 200 300>)
     Index=$System.GameModeID
     Num=::NewEnumerator2
    ))
   &ShapeLocation(
    MaskSpawnGroup=false
    Override Seed=false
    Sphere Radius=&RandomRangeFloat(Minimum=&Multiply_Float(A=5, B=2), Maximum=&Multiply_Float(A=10, B=2))
    UseOwnerScale=true
   )
   /VehicleCosmetics/FX/VC_FXParents/Modules/DLR_ProductPreviews(GameModeID:NiagaraInt32=$System.GameModeID, bPreviewVehicleSpeed:NiagaraBool=$System.bPreviewVehicleSpeed)
  ]
  particleUpdateScript [
   &ParticleState()
   &ScaleSpriteSize(Uniform Curve Sprite Scale=DataInterface() {propertyValues {Curve {keys [
        {tangentMode=RCTM_User, time=0, value=1}
        {interpMode=RCIM_Cubic, tangentMode=RCTM_Break, time=0.7, value=1}
        {interpMode=RCIM_Cubic, tangentMode=RCTM_Break, time=0.95, value=0}
       ]}}})
   &InheritSourceMovement(
    Applied Position Scale=&Multiply_VectorByFloat(
     Vector=<1 1 1>
     Float=&FloatFromCurve(FloatCurve=DataInterface() {propertyValues {Curve {
         defaultValue=1
         keys [
          {arriveTangent=-1.7885, interpMode=RCIM_Cubic, leaveTangent=-1.7885, tangentMode=RCTM_User, time=0.05, value=1}
          {interpMode=RCIM_Cubic, tangentMode=RCTM_User, time=1, value=0}
         ]
        }}})
    )
    ApplyForce=false
    ApplyPosition=true
    ApplyPositionFalloff=false
   ) {module=InheritSourceMovement001}
   &SubUVAnimation(
    "Sprite Renderer "=DataInterface() {propertyValues {SpriteRenderer=&self:Hearts_0.NiagaraSpriteRendererProperties_5}}
    SubUV Lookup Index Scale=-2
    UseEndFrame=false
    UseStartFrame=false
   )
   &DynamicMaterialParameters(
    Index 0 Param 1=1.1
    Index 0 Param 2=1
    Index 0 Param 3=1
    Module.Index0Param0Write=true
    Module.Index0Param1Write=true
    Module.Index0Param2Write=true
    Module.Index0Param3Write=true
   )
   &SolveForcesAndVelocity(Clamp Velocity=false, Limit Acceleration=false)
  ]
  rendererClasses [&NiagaraSpriteRendererProperties]
  renderers [
   {rendererClass=&NiagaraSpriteRendererProperties}
   {rendererClass=&NiagaraSpriteRendererProperties, rendererIndex=1}
  ]
 }
 NE_VC_Boost_Ribbon_Parent {
  emitterUpdateScript [
   &EmitterState()
   &VC_FilteredSocketCount()
   &VC_RateQualityHandler(Epic Spawn Rate=30, High Spawn Rate=30, Low Spawn Rate=30, Use Custom Rate?=true)
   &VC_SpawnRibbon_v1(Number of Ribbons=$Emitter.PortCount, SpawnRate=$Emitter.Spawn Rate)
   /SetVariables(Emitter.ContrastFadeFront:NiagaraFloat=&MergeHelper_FloatSwitch(0=1, 1=1, 2=1, 3=0.7, Index=$System.GameModeID, Num=::NewEnumerator2)) {isSetParameters=true}
  ]
  particleSpawnScript [
   &VC_PaintedVariantsOverride() {enabled=false}
   &VC_InitializeRibbon_v1(
    Color=<rgba 0.9792 0 0.7921 1>
    Custom Offset=true
    Enable Skeletal Position=true
    Lifetime=1
    Local Vector Offset=&AddVectorToPosition(
     Position=&MergeHelper_VectorSwitch(0=<0 0 0>, 1=<0 0 0>, 2=<0 0 0>, 3=<0 0 0>, Index=$System.GameModeID, Num=::NewEnumerator2)
     Vector=$User.DTCarOffset
    )
    Offset Amount=1
    Ribbon Width=15
   )
   &AddVelocity(Velocity=&MergeHelper_VectorSwitch(0=<-2000 0 0>, 1=<-2000 0 0>, 2=<-2000 0 0>, 3=<-2000 0 0>, Index=$System.GameModeID, Num=::NewEnumerator2)) {module=AddVelocity001}
   &DLR_ProductPreviews(GameModeID:NiagaraInt32=$System.GameModeID, bPreviewVehicleSpeed:NiagaraBool=$System.bPreviewVehicleSpeed)
  ]
  particleUpdateScript [
   &ParticleState()
   &InheritSourceMovement(
    Applied Position Scale=&VectorFromFloat(Value=&FloatFromCurve(FloatCurve=DataInterface() {propertyValues {Curve {
         defaultValue=1
         keys [
          {arriveTangent=-1.7885, interpMode=RCIM_Cubic, leaveTangent=-1.7885, tangentMode=RCTM_User, time=0.05, value=1}
          {interpMode=RCIM_Cubic, tangentMode=RCTM_User, time=1, value=0}
         ]
        }}}))
    ApplyForce=false
    ApplyPosition=true
    ApplyPositionFalloff=false
   )
   &SolveForcesAndVelocity(Clamp Velocity=false, Limit Acceleration=false)
   &VC_UpdateRibbon_v1(Enable Skeletal Position=true)
   /SetVariables(
    Particles.PINKWPO:NiagaraPosition=&AddVectorToPosition(
     Position=$Particles.Position
     Vector=&TransformVector(
      Vector=&MakeVector(
       X=0
       Y=&Cosine(
        Normalized Angle=&Add_Float(
         A=&Multiply_Float(A=$Particles.NormalizedAge, B=-2)
         B=&Multiply_Float(A=$Emitter.LoopedAge, B=2)
        )
        Period=1
        Scale=&FloatFromCurve(FloatCurve=DataInterface() {propertyValues {Curve {keys [<at 0 1.25 RCIM_Cubic>,<at 0.5 1.5 RCIM_Cubic>]}}}, Scale Curve=10)
        Bias=0
       )
       Z=&Sine(
        Normalized Angle=&Add_Float(
         A=&Multiply_Float(A=$Particles.NormalizedAge, B=-2)
         B=&Multiply_Float(A=$Emitter.LoopedAge, B=2)
        )
        Period=1
        Scale=&FloatFromCurve(
         FloatCurve=DataInterface() {propertyValues {Curve {keys [
             {arriveTangentWeight=0.1667, interpMode=RCIM_Cubic, leaveTangentWeight=0.1667, tangentMode=RCTM_User, time=0, value=1.25}
             {arriveTangentWeight=0.1667, interpMode=RCIM_Cubic, leaveTangentWeight=0.1667, tangentMode=RCTM_User, time=1, value=1.5}
            ]}}}
         Scale Curve=10
        )
        Bias=0
       )
      )
      Destination Space=::Simulation
      Apply Scale=true
     )
    )
    Particles.PurpleWPO:NiagaraPosition=&AddVectorToPosition(
     Position=$Particles.Position
     Vector=&TransformVector(
      Vector=&MakeVector(
       X=0
       Y=&Cosine(
        Normalized Angle=&Add_Float(
         A=&Multiply_Float(A=$Particles.NormalizedAge, B=-2)
         B=&Multiply_Float(A=$Emitter.LoopedAge, B=4)
        )
        Period=0.9
        Scale=&FloatFromCurve(FloatCurve=DataInterface() {propertyValues {Curve {keys [<at 0 1.5 RCIM_Cubic>,<at 1 1.5 RCIM_Cubic>]}}}, Scale Curve=20)
        Bias=0
       )
       Z=&Sine(
        Normalized Angle=&Add_Float(
         A=&Multiply_Float(A=$Particles.NormalizedAge, B=-2)
         B=&Multiply_Float(A=$Emitter.LoopedAge, B=4)
        )
        Period=0.9
        Scale=&FloatFromCurve(FloatCurve=DataInterface() {propertyValues {Curve {keys [<at 0 1.5 RCIM_Cubic>,<at 1 1.5 RCIM_Cubic>]}}}, Scale Curve=20)
        Bias=0
       )
      )
      Destination Space=::Simulation
      Apply Scale=true
     )
    )
    Particles.RedWPO:NiagaraPosition=&AddVectorToPosition(
     Position=$Particles.Position
     Vector=&TransformVector(
      Vector=&MakeVector(
       X=0
       Y=&Cosine(
        Normalized Angle=&Add_Float(
         A=&Multiply_Float(A=$Particles.NormalizedAge, B=-1.5)
         B=&Multiply_Float(A=$Emitter.LoopedAge, B=3)
        )
        Period=0.9
        Scale=&FloatFromCurve(FloatCurve=DataInterface() {propertyValues {Curve {keys [<at 0 1.8 RCIM_Cubic>,<at 1 1.8 RCIM_Cubic>]}}}, Scale Curve=15)
        Bias=0
       )
       Z=&Sine(
        Normalized Angle=&Add_Float(
         A=&Multiply_Float(A=$Particles.NormalizedAge, B=-1.5)
         B=&Multiply_Float(A=$Emitter.LoopedAge, B=4)
        )
        Period=0.9
        Scale=&FloatFromCurve(FloatCurve=DataInterface() {propertyValues {Curve {keys [<at 0 1.8 RCIM_Cubic>,<at 1 1.8 RCIM_Cubic>]}}}, Scale Curve=15)
        Bias=0
       )
      )
      Destination Space=::Simulation
      Apply Scale=true
     )
    )
   ) {isSetParameters=true}
   &ScaleRibbonWidth(Ribbon Width Scale=&FloatFromCurve(FloatCurve=DataInterface() {propertyValues {Curve {keys [<at 0.2 0.8 RCIM_Cubic>,<at 1 0 RCIM_Cubic>]}}}))
   /SetVariables(Particles.R2:LinearColor=<rgba 0.2696 0.1196 0.6562 1>, Particles.R3:LinearColor=<rgba 0.7917 0.2047 0.2543 1>) {isSetParameters=true}
  ]
  rendererClasses [&NiagaraRibbonRendererProperties]
  renderers [
   {rendererClass=&NiagaraRibbonRendererProperties}
   {rendererClass=&NiagaraRibbonRendererProperties, rendererIndex=1}
   {rendererClass=&NiagaraRibbonRendererProperties, rendererIndex=2}
  ]
 }
}
summary {
 systemName=NS_Boost_Hearts
 emitters [
  {emitterName=Hearts, rendererClasses [&NiagaraSpriteRendererProperties]}
  {emitterName=NE_VC_Boost_Ribbon_Parent, rendererClasses [&NiagaraRibbonRendererProperties]}
 ]
 userVariables [
  User.bIsFrontendPreview {type=NiagaraBool}
  User.bIsPainted {type=NiagaraBool}
  User.blsIsCoroColor {type=NiagaraBool}
  User.bPreviewVehicleSpeed {type=NiagaraBool}
  User.ColorVariantID {type=NiagaraInt32}
  User.DTCarOffset {type=Vector3f, description="", defaultValue {struct=Vector3f, value=<0 0 0>}}
  User.GameModeID {type=NiagaraInt32, defaultValue=3}
  User.PaintedColor {type=LinearColor, description="", defaultValue {struct=LinearColor, value=<rgba 1 1 1 1>}}
  User.PaintedScalar {type=NiagaraFloat, defaultValue=1, description=""}
  User.PortCount {type=NiagaraInt32, defaultValue=2}
 ]
}
